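# Post-process-style AnimBlueprint for the MetalScout backpack, adjusting the pack bone to fit the wearer's body and pose.
# A Look At toward the body's pelvis feeds chained Transform Bone offsets for base fit, large bodies, skydiving, melee weapons and left-clavicle movement.
# Update logic supplies these inputs, deriving skydive offsets from spine yaw.
# Weapon and skydive events toggle material "Hide Element 10", hiding it while a MetalScout-tagged pickaxe is equipped outside skydiving.
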AnimBlueprint /BRCosmetics/Accessories/FORT_Backpacks/Backpack_M_MED_MetalScout/Meshes/M_MED_MetalScout_Pack_AnimBP
target_skeleton=&/M_MED_MetalScout_Pack_Skeleton
graphs [
 AnimGraph {nodes [
   n1 = Input Pose()
   n2 = "Transform (Modify) Bone - Bone: pack"(
    ComponentPose=Look At - Bone: pack!(ComponentPose=Local To Component(LocalPose=@n1), Alpha=1, LookAtLocation:Vector=Get PelviceLocation() -> (PelviceLocation:Vector)) -> (Pose) {alphaBoolBlend {blendOption=Linear}, boneToModify {boneName=pack}, lookAt_Axis {axis {y=0, z=-1}}, lookAtTarget=~}
    Translation="0,-1.5,0"
    Rotation="0,0,25"
   ) {
    rotationMode=BMM_Additive
    rotationSpace=BCS_ParentBoneSpace
    translationMode=BMM_Additive
    translationSpace=BCS_BoneSpace
    alphaBoolBlend {blendOption=Linear}
    boneToModify {boneName=pack}
    rotation {roll=25}
    translation {y=-1.5}
   }
   n3 = "Transform (Modify) Bone - Bone: pack"!(ComponentPose=@n2, bAlphaBoolEnabled:bool=Get LargeBodyType() -> (LargeBodyType:bool), Translation="0,-2,0", Rotation="0,0,6", Scale) -> (Pose) {
    alphaInputType=Bool
    rotationMode=BMM_Additive
    rotationSpace=BCS_ParentBoneSpace
    alphaBoolBlend {blendOption=Linear}
    boneToModify {boneName=pack}
    rotation {roll=6}
    translation {y=-2}
   }
   n4 = "Transform (Modify) Bone - Bone: pack"(
    ComponentPose=@n3
    bAlphaBoolEnabled=Get SkydiveOffset() -> (SkydiveOffset:bool)
    Translation=Get BackpackTranslate() -> (BackpackTranslate:Vector)
    Rotation=Get BackpackRotate() -> (BackpackRotate:Rotator)
   ) {
    alphaInputType=Bool
    rotationMode=BMM_Additive
    rotationSpace=BCS_ParentBoneSpace
    translationMode=BMM_Additive
    translationSpace=BCS_ParentBoneSpace
    alphaBoolBlend {blendInTime=0.1, blendOption=Linear, blendOutTime=0.2}
    boneToModify {boneName=pack}
    rotation {pitch=-10, roll=-8}
    translation {y=-6, z=-8}
   }
   n5 = "Transform (Modify) Bone - Bone: pack"(ComponentPose=@n4, bAlphaBoolEnabled=Get PickaxeEquped() -> (PickaxeEquped:bool), Translation="2,0,0", Rotation="0,0,0") {
    alphaInputType=Bool
    rotationSpace=BCS_ParentBoneSpace
    translationMode=BMM_Additive
    translationSpace=BCS_ParentBoneSpace
    alphaBoolBlend {blendInTime=0.1, blendOption=Linear, blendOutTime=0.1}
    boneToModify {boneName=pack}
    translation {x=2}
   }
   n6 = "Transform (Modify) Bone - Bone: pack"!(ComponentPose=@n5, Alpha:float=Get ClavicleAlpha() -> (ClavicleAlpha:float), Translation="0,0,0", Rotation="0,-10,0", Scale) -> (Pose) {
    rotationMode=BMM_Additive
    rotationSpace=BCS_ParentBoneSpace
    alphaBoolBlend {blendOption=Linear}
    boneToModify {boneName=pack}
    rotation {yaw=-10}
   }
   Output Pose(Result=Component To Local(ComponentPose=@n6)) {layerGroup=None}
  ]}
 EventGraph {nodes [
   Event BlueprintInitializeAnimation()
   > n1 = Cast To PlayerPawn_Athena(Object=TryGetPawnOwner())
   > Set LargeBodyType(LargeBodyType:bool="Equal (Enum)"(
     A:EFortCustomBodyType Enum=GetCharacterBodyType!(self:Fort Player Pawn Object Reference=@n1) -> (ReturnValue:EFortCustomBodyType Enum=NONE)
     B:EFortCustomBodyType Enum=Large
    )) -> (Output_Get:bool)
   Event BlueprintUpdateAnimation()
   > n2 = Sequence!(execute) -> (then_0, then_1, then_2:Exec)
   > Set SkydiveOffset(SkydiveOffset:bool=OR Boolean(A=Get bIsSkydiveDiveMode(), B=Get bIsSkydiveFloating())) -> (Output_Get:bool)
   n3 = Get BodySkelMesh() -> (BodySkelMesh:Skeletal Mesh Component Object Reference)
   > Is Valid(InputObject=@n3)
   > Set PelviceLocation(PelviceLocation:Vector=GetSocketLocation!(self:Scene Component Object Reference=@n3, InSocketName:Name=pelvis) -> (ReturnValue:Vector)) -> (Output_Get:Vector)
   > n4 = Cast To FortPlayerAnimInstance(Object=GetAnimInstance(self=GetOwnerSkeletalMeshComponent()))
   > n5 = Switch on EFortWeaponCoreAnimation!(execute, Selection:EFortWeaponCoreAnimation Enum=Get WeaponCoreAnim(self=@n4)) -> (
    Melee:Exec
    Pistol:Exec
    Shotgun:Exec
    PaperBlueprint:Exec
    Rifle:Exec
    MeleeOneHand:Exec
    MachinePistol:Exec
    RocketLauncher:Exec
    GrenadeLauncher:Exec
    GoingCommando:Exec
    AssaultRifle:Exec
    TacticalShotgun:Exec
    SniperRifle:Exec
    TrapPlacement:Exec
    ShoulderLauncher:Exec
    AbilityDecoTool:Exec
    Crossbow:Exec
    C4:Exec
    RemoteControl:Exec
    DualWield:Exec
    AR_BullPup:Exec
    AR_ForwardGrip:Exec
    MedPackPaddles:Exec
    SMG_P90:Exec
    AR_DrumGun:Exec
    Consumable_Small:Exec
    Consumable_Large:Exec
    Balloon:Exec
    MountedTurret:Exec
    CreativeTool:Exec
    ExplosiveBow:Exec
    AshtonIndigo:Exec
    AshtonChicago:Exec
    MeleeDualWield:Exec
    Unarmed:Exec
   )
   Set PickaxeEquped(
    execute=[
     @n5.AR_BullPup
     @n5.AR_DrumGun
     @n5.AR_ForwardGrip
     @n5.AbilityDecoTool
     @n5.AshtonChicago
     @n5.AshtonIndigo
     @n5.AssaultRifle
     @n5.Balloon
     @n5.C4
     @n5.Consumable_Large
     @n5.Consumable_Small
     @n5.CreativeTool
     @n5.Crossbow
     @n5.DualWield
     @n5.ExplosiveBow
     @n5.GoingCommando
     @n5.GrenadeLauncher
     @n5.MachinePistol
     @n5.MedPackPaddles
     @n5.MountedTurret
     @n5.PaperBlueprint
     @n5.Pistol
     @n5.RemoteControl
     @n5.Rifle
     @n5.RocketLauncher
     @n5.SMG_P90
     @n5.Shotgun
     @n5.ShoulderLauncher
     @n5.SniperRifle
     @n5.TacticalShotgun
     @n5.TrapPlacement
     @n5.Unarmed
    ]
    PickaxeEquped:bool=false
   ) -> (Output_Get:bool)
   Set PickaxeEquped(execute=[@n5.Melee, @n5.MeleeDualWield, @n5.MeleeOneHand], PickaxeEquped=true)
   Set ClavicleAlpha(
    execute=@n2.then_1
    ClavicleAlpha:float=GetAlphaFromReferenceTransformDelta(SkeletalMeshComponent=Get BodySkelMesh(), BoneName=clavicle_l, BaseName=spine_05, InputRange=(X=-10,Y=-23), TwistAxis="1,0,0")
   ) -> (Output_Get:float)
   n6 = GetOwnerSkeletalMeshComponent()
   n7 = BreakRotator(InRot=BreakTransform(InTransform=GetDeltaTransformFromRefPose(self=@n6, BoneName=spine_05, BaseName=spine_01)).Rotation)
   Set BackpackRotate(execute=@n2.then_2, BackpackRotate:Rotator=MakeRotator(Roll=MapRangeClamped(Value=@n7.Yaw, InRangeA=0, InRangeB=-30, OutRangeB=-10))) -> (Output_Get:Rotator)
   > Set BackpackTranslate(BackpackTranslate:Vector=MakeVector(X=MapRangeClamped(Value=@n7.Yaw, InRangeA=0, InRangeB=-30, OutRangeB=-5), Y=-10, Z=-5)) -> (Output_Get:Vector)
   n8 = Event OnWeaponChanged()
   n9 = Get Dynamic Material Instance(execute=[@n8.then, @n16.then_0], SkeletalMeshComponent=GetPartSkeletalMeshComponent())
   n10 = PickaxeItemDefinitionHasAnyMetaTag(Tags=Make Array("[0]"=(TagName="Cosmetics.MetaTags.MetalScout")), bRequiresEquipped=true)
   > n11 = Branch(Condition=@n10)
   > n12 = Branch(Condition=Get bIsSkyDiving())
   SetScalarParameterValue(execute=[@n13.then, @n12.then, @n14.else, @n11.else], self=@n9, ParameterName=Hide Element 10, Value=0)
   SetScalarParameterValue(execute=[@n12.else, @n14.then], self=@n9, ParameterName=Hide Element 10, Value=1)
   n13 = Event OnBeginSkydiving()
   Event OnEndSkydiving()
   > n14 = Branch(Condition=@n10)
   n15 = Event OnCharacterCustomizationLoaded()
   > n16 = Sequence!(execute) -> (then_0, then_1)
   Set BodySkelMesh(execute=@n16.then_1, BodySkelMesh:Skeletal Mesh Component Object Reference=@n15.CustomBaseMesh) -> (Output_Get:Skeletal Mesh Component Object Reference)
   GetSkeletalMeshForPartType(self=@n1)
  ]}
]
variables [
 BackpackRotate {type=Rotator}
 BackpackTranslate {type=Vector}
 BodySkelMesh {type=SkeletalMeshComponent}
 ClavicleAlpha {type=number}
 LargeBodyType {type=boolean}
 PelviceLocation {type=Vector}
 PickaxeEquped {type=boolean}
 SkydiveOffset {type=boolean}
]
